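; Base local gameplay behavior for Athena AI bots that provides helpers to make a bot face the direction of the behavior actor.
; SetFocusRotation validates the bot controller and its pawn, then places the focal point out from the pawn's location along this actor's forward vector.
; ResetFocusRotation clears the controller's focus.
; The events and construction script are empty, and neither helper is called here.
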(event EventBeginPlay)

(event Collision|EventActorBeginOverlap
  (params (OtherActor :type Actor%20Object%20Reference)))

(event EventTick
  (params (DeltaSeconds :type Float%20%28single-precision%29)))

(event LocalGameplayBehavior|EventOnExecute
  (params (Activator :type Actor%20Object%20Reference)))

(event LocalGameplayBehavior|EventOnAbort)

(fn ResetFocusRotation
  (params (BotController :type Fort%20Athena%20AIBot%20Controller%20Object%20Reference))
  (Utilities|IsValid BotController
    (:Is%20Valid
      (AI|ClearFocus BotController))))

(fn SetFocusRotation
  (params (BotController :type Fort%20Athena%20AIBot%20Controller%20Object%20Reference))
  (bind _playerbotpawn (Variables|/Script/FortniteAI.FortAthenaAIBotController|GetPlayerBotPawn BotController))
  (Utilities|IsValid BotController
    (:Is%20Valid
      (Utilities|IsValid _playerbotpawn
        (:Is%20Valid
          (AI|SetFocalPoint BotController (+ (Transformation|GetActorLocation _playerbotpawn) (* (Transformation|GetActorForwardVector self) 1e+02))))))))

(fn ConstructionScript)
