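; Local gameplay behavior for patrols that, by random chance, makes an Athena AI bot wait a random time or finishes immediately.
; If the direction arrow exists, waiting also focuses the bot along the actor's forward vector and plays a random configured speech bubble.
; Construction flattens rotation to yaw and destroys the arrow when LookWhileWaiting is off, disabling both look and speech.
; Focus clears after the delay or on abort.
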(event LocalGameplayBehavior|EventOnExecute
  (params (Activator :type Actor%20Object%20Reference))
  (bind _speechbubbletags (Variables|Setup|SpeechBubbles|GetSpeechBubbleTags))
  (if (<= (Math|Random|RandomFloatinRange 0.0 1.0) (Variables|Setup|GetChanceToWait))
    (Utilities|FlowControl|Sequence
      (:then_0
        (Utilities|IsValid (Variables|Default|GetDirectionArrow)
          (:Is%20Valid
            (bind _asfort_athena_aibot_controller (Utilities|Casting|CastToFortAthenaAIBotController Activator))
            (bind _output_get (Variables|Default|SetAIBotControllerActivator _asfort_athena_aibot_controller))
            (CallFunction|SetFocalPoint)
            (if (Utilities|Array|IsNotEmpty _speechbubbletags)
              (bind _asfort_pawn (Utilities|Casting|CastToFortPawn (Pawn|GetControlledPawn _output_get)))
              (Class|BPFLAICommonFunctions|PlaySpeechBubble _asfort_pawn (Utilities|Array|Get(aref) _speechbubbletags (Math|Random|RandomIntegerinRange 0 (- (Utilities|Array|Length _speechbubbletags) 1))) (Variables|Setup|SpeechBubbles|GetStop%20Previous%20Line) (Variables|Setup|SpeechBubbles|GetTry%20Play%20VOLine))))))
      (:then_1
        (Utilities|FlowControl|Delay (Math|Random|RandomFloatinRange (Variables|Setup|GetWaitTimeMin) (Variables|Setup|GetWaitTimeMax)))
        (CallFunction|ResetFocalPoint)
        (LocalGameplayBehavior|FinishExecute)))
    (else
      (LocalGameplayBehavior|FinishExecute))))

(event LocalGameplayBehavior|EventOnAbort
  (CallFunction|ResetFocalPoint)
  (Variables|Default|SetAIBotControllerActivator 0))

(fn ResetFocalPoint
  (bind _aibotcontrolleractivator (Variables|Default|GetAIBotControllerActivator)
    (:then
      (AI|ClearFocus _aibotcontrolleractivator))
    (:else
      _))
  (return))

(fn SetFocalPoint
  (bind _aibotcontrolleractivator (Variables|Default|GetAIBotControllerActivator)
    (:then
      (AI|SetFocalPoint _aibotcontrolleractivator (+ (Transformation|GetActorLocation (Variables|/Script/FortniteAI.FortAthenaAIBotController|GetPlayerBotPawn _aibotcontrolleractivator)) (* (Transformation|GetActorForwardVector self) 1e+02))))
    (:else)))

(fn ConstructionScript
  (Transformation|SetActorRotation (Math|Rotator|MakeRotator 0.0 0.0 (.yaw (Transformation|GetActorRotation)))) ; Make sure we don't pitch or roll the actor. This whole thing is to set YAW of the AI
  (if (not (Variables|Setup|GetLookWhileWaiting))
    (Components|DestroyComponent (Variables|Default|GetDirectionArrow))))
