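; Base Local Gameplay Behavior that makes an Athena AI bot play random emotes from a configurable list.
; On execute, a chance roll and stressed-tag check gate the emote, with optional delay and focal point along the behavior actor's forward vector.
; Each emote runs for its duration with an optional global or per-emote GameplayEffect, then loops via PlayEmote or finishes.
; Abort stops the emote, clears focus and removes the effect unless configured to keep it.
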(event EventBeginPlay)

(event Collision|EventActorBeginOverlap
  (params (OtherActor :type Actor%20Object%20Reference)))

(event EventTick
  (params (DeltaSeconds :type Float%20%28single-precision%29)))

(event Custom|PlayEmote
  (bind _aibotcontrolleractivator (Variables|Default|GetAIBotControllerActivator))
  (bind _playerbotpawn (Variables|/Script/FortniteAI.FortAthenaAIBotController|GetPlayerBotPawn _aibotcontrolleractivator))
  (bind _useglobalgameplayeffectforallemotes (Variables|EmoteSetup|GetUseGlobalGameplayEffectForAllEmotes))
  (bind _globalgameplayeffect (Variables|EmoteSetup|GetGlobalGameplayEffect))
  ;[1 1 1] Select and Play Emote
  (bind _output_get (Variables|Default|SetSelectedEmote (Utilities|Array|Get(aref) (Variables|EmoteSetup|GetEmoteList) (Math|Random|RandomIntegerinRange 0 (- (Utilities|Array|Length (Variables|EmoteSetup|GetEmoteList)) 1)))))
  (Class|FortAthenaAIBotController|PlayEmote (Variables|Default|GetAIBotControllerActivator) (Utilities|Struct|BreakSTRUCTLGBPlayEmoteList _output_get))
  (if (Utilities|IsValidClass (select _useglobalgameplayeffectforallemotes _globalgameplayeffect (.GEToApplyDuringEmote_9_3E730CB446C893E4087B409AC89A9B33 (Utilities|Struct|BreakSTRUCTLGBPlayEmoteList _output_get))))
    (bind _returnvalue (Environment|ApplyGlobalEnvironmentGameplayEffecttoActor (Variables|/Script/FortniteAI.FortAthenaAIBotController|GetPlayerBotPawn (Variables|Default|GetAIBotControllerActivator)) (select _useglobalgameplayeffectforallemotes _globalgameplayeffect (.GEToApplyDuringEmote_9_3E730CB446C893E4087B409AC89A9B33 (Utilities|Struct|BreakSTRUCTLGBPlayEmoteList _output_get)))))
    (Variables|Default|SetHAND_GEToApply _returnvalue))
  (Utilities|FlowControl|Delay (.EmoteDuration_5_17E1ADA942EAA53EF5DD61AEF468C9B6 (Utilities|Struct|BreakSTRUCTLGBPlayEmoteList _output_get)))
  (Class|FortAthenaAIBotController|StopEmote _aibotcontrolleractivator)
  (if (Math|Boolean|ANDBoolean (Utilities|IsValidClass (select _useglobalgameplayeffectforallemotes _globalgameplayeffect (.GEToApplyDuringEmote_9_3E730CB446C893E4087B409AC89A9B33 (Utilities|Struct|BreakSTRUCTLGBPlayEmoteList _output_get)))) (Utilities|IsValid _playerbotpawn :!class /Script/Engine.KismetSystemLibrary) (not (Variables|EmoteSetup|GetShouldKeepGameplayEffectOnEndOrAbort)))
    (GameplayEffects|RemoveActiveGameplayEffect (Ability|GetAbilitySystemComponent _playerbotpawn) (Variables|Default|GetHAND_GEToApply) -1)
    (else
      _))
  (if (Variables|EmoteSetup|GetLoopAtEnd)
    (CallFunction|PlayEmote)
    (else
      (LocalGameplayBehavior|FinishExecute)
      (CallFunction|ResetFocalPoint)
      (Variables|Default|SetAIBotControllerActivator 0))))

(event LocalGameplayBehavior|EventOnExecute
  (params (Activator :type Actor%20Object%20Reference))
  (bind _aibotcontrolleractivator (Variables|Default|GetAIBotControllerActivator))
  (bind _playerbotpawn (Variables|/Script/FortniteAI.FortAthenaAIBotController|GetPlayerBotPawn _aibotcontrolleractivator))
  (bind _useglobalgameplayeffectforallemotes (Variables|EmoteSetup|GetUseGlobalGameplayEffectForAllEmotes))
  (bind _globalgameplayeffect (Variables|EmoteSetup|GetGlobalGameplayEffect))
  (if (<= (Math|Random|RandomFloatinRange 0.0 1.0) (Variables|EmoteSetup|GetChanceToPlayEmote))
    (bind _asfort_athena_aibot_controller (Utilities|Casting|CastToFortAthenaAIBotController Activator)
      (:then
        (bind _output_get (Variables|Default|SetAIBotControllerActivator _asfort_athena_aibot_controller))
        (if (GameplayTags|HasMatchingGameplayTag (Variables|/Script/FortniteAI.FortAthenaAIBotController|GetPlayerBotPawn _output_get) (Variables|Default|GetTC_NPCStatusStressed))
          _
          (elif (Variables|EmoteSetup|GetUseDirection)
            (CallFunction|SetFocalPoint)
            ;[1 1 1] Validate if we play Delay before Emoting
            (if (> (Variables|EmoteSetup|GetDelayBeforeEmote) 0.0)
              (Utilities|FlowControl|Delay (Variables|EmoteSetup|GetDelayBeforeEmote))
              (else
                _))
            ;[1 1 1] Select and Play Emote
            (bind _output_get_2 (Variables|Default|SetSelectedEmote (Utilities|Array|Get(aref) (Variables|EmoteSetup|GetEmoteList) (Math|Random|RandomIntegerinRange 0 (- (Utilities|Array|Length (Variables|EmoteSetup|GetEmoteList)) 1)))))
            (Class|FortAthenaAIBotController|PlayEmote (Variables|Default|GetAIBotControllerActivator) (Utilities|Struct|BreakSTRUCTLGBPlayEmoteList _output_get_2))
            (if (Utilities|IsValidClass (select _useglobalgameplayeffectforallemotes _globalgameplayeffect (.GEToApplyDuringEmote_9_3E730CB446C893E4087B409AC89A9B33 (Utilities|Struct|BreakSTRUCTLGBPlayEmoteList _output_get_2))))
              (bind _returnvalue (Environment|ApplyGlobalEnvironmentGameplayEffecttoActor (Variables|/Script/FortniteAI.FortAthenaAIBotController|GetPlayerBotPawn (Variables|Default|GetAIBotControllerActivator)) (select _useglobalgameplayeffectforallemotes _globalgameplayeffect (.GEToApplyDuringEmote_9_3E730CB446C893E4087B409AC89A9B33 (Utilities|Struct|BreakSTRUCTLGBPlayEmoteList _output_get_2)))))
              (Variables|Default|SetHAND_GEToApply _returnvalue))
            (Utilities|FlowControl|Delay (.EmoteDuration_5_17E1ADA942EAA53EF5DD61AEF468C9B6 (Utilities|Struct|BreakSTRUCTLGBPlayEmoteList _output_get_2)))
            (Class|FortAthenaAIBotController|StopEmote _aibotcontrolleractivator)
            (if (Math|Boolean|ANDBoolean (Utilities|IsValidClass (select _useglobalgameplayeffectforallemotes _globalgameplayeffect (.GEToApplyDuringEmote_9_3E730CB446C893E4087B409AC89A9B33 (Utilities|Struct|BreakSTRUCTLGBPlayEmoteList _output_get_2)))) (Utilities|IsValid _playerbotpawn :!class /Script/Engine.KismetSystemLibrary) (not (Variables|EmoteSetup|GetShouldKeepGameplayEffectOnEndOrAbort)))
              (GameplayEffects|RemoveActiveGameplayEffect (Ability|GetAbilitySystemComponent _playerbotpawn) (Variables|Default|GetHAND_GEToApply) -1)
              (else
                _))
            (if (Variables|EmoteSetup|GetLoopAtEnd)
              (CallFunction|PlayEmote)
              (else
                (LocalGameplayBehavior|FinishExecute)
                (CallFunction|ResetFocalPoint)
                (Variables|Default|SetAIBotControllerActivator 0)))
            (else
              _))))
      (:CastFailed
        _))
    (LocalGameplayBehavior|FinishExecute)
    (else
      _)))

(event LocalGameplayBehavior|EventOnAbort
  (bind _aibotcontrolleractivator (Variables|Default|GetAIBotControllerActivator))
  (bind _playerbotpawn (Variables|/Script/FortniteAI.FortAthenaAIBotController|GetPlayerBotPawn _aibotcontrolleractivator))
  (Variables|Default|GetAIBotControllerActivator
    (:then
      (if (Math|Boolean|ANDBoolean (Utilities|IsValidClass (select (Variables|EmoteSetup|GetUseGlobalGameplayEffectForAllEmotes) (Variables|EmoteSetup|GetGlobalGameplayEffect) (.GEToApplyDuringEmote_9_3E730CB446C893E4087B409AC89A9B33 (Utilities|Struct|BreakSTRUCTLGBPlayEmoteList (stale Custom|PlayEmote _output_get))))) (Utilities|IsValid _playerbotpawn :!class /Script/Engine.KismetSystemLibrary) (not (Variables|EmoteSetup|GetShouldKeepGameplayEffectOnEndOrAbort)))
        (GameplayEffects|RemoveActiveGameplayEffect (Ability|GetAbilitySystemComponent _playerbotpawn) (Variables|Default|GetHAND_GEToApply) -1)
        (else
          _))
      (Class|FortAthenaAIBotController|StopEmote (Variables|Default|GetAIBotControllerActivator))
      (CallFunction|ResetFocalPoint))
    (:else)))

(fn ResetFocalPoint
  (if (Variables|EmoteSetup|GetUseDirection)
    (bind _aibotcontrolleractivator (Variables|Default|GetAIBotControllerActivator)
      (:then
        (AI|ClearFocus _aibotcontrolleractivator))
      (:else))
    (else
      _))
  (return))

(fn SetFocalPoint
  (bind _aibotcontrolleractivator (Variables|Default|GetAIBotControllerActivator)
    (:then
      (AI|SetFocalPoint _aibotcontrolleractivator (+ (Transformation|GetActorLocation (Variables|/Script/FortniteAI.FortAthenaAIBotController|GetPlayerBotPawn _aibotcontrolleractivator)) (* (Transformation|GetActorForwardVector self) 1e+02))))
    (:else)))

(fn ConstructionScript)
