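; Local gameplay behavior that makes an activating Fortnite Athena AI bot act out a solo story as a timed sequence of emotes.
; On execute it optionally points the bot along this actor's forward vector, then StartStory steps through EmoteList, re-calling itself after each duration while PlayEmote starts and stops each emote.
; At list end it loops after a delay or finishes.
; Finishing and aborting stop the emote, clear focus and release the bot.
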(event EventBeginPlay)

(event Collision|EventActorBeginOverlap
  (params (OtherActor :type Actor%20Object%20Reference)))

(event EventTick
  (params (DeltaSeconds :type Float%20%28single-precision%29)))

(event LocalGameplayBehavior|EventOnExecute
  (params (Activator :type Actor%20Object%20Reference))
  (bind _asfort_athena_aibot_controller (Utilities|Casting|CastToFortAthenaAIBotController Activator)
    (:then
      (Variables|Default|SetAIBotControllerActivator _asfort_athena_aibot_controller)
      (if (Variables|EmoteSetup|GetUseDirection)
        (CallFunction|SetFocalPoint)
        (else
          _))
      (Utilities|FlowControl|Delay 0.6)
      (CallFunction|StartStory))
    (:CastFailed
      (LocalGameplayBehavior|FinishExecute))))

(event LocalGameplayBehavior|EventOnAbort
  (bind _aibotcontrolleractivator (Variables|Default|GetAIBotControllerActivator)
    (:then
      (Class|FortAthenaAIBotController|StopEmote _aibotcontrolleractivator)
      (CallFunction|ResetFocalPoint)
      (Variables|Default|SetAIBotControllerActivator 0)
      (LocalGameplayBehavior|FinishExecute))
    (:else)))

(event Custom|StartStory
  (bind _currentemoteindex (Variables|Default|GetCurrentEmoteIndex))
  (bind _emotelist (Variables|EmoteSetup|GetEmoteList))
  (bind _output (Utilities|Array|Get(aref) _emotelist _currentemoteindex))
  (bind (_emoteprimaryassetid_2_4fac2b88497d53109f4d1c94ce5db825 _emoteduration_5_17e1ada942eaa53ef5dd61aef468c9b6) (Utilities|Struct|BreakSTRUCTLGBStoryPairedEmoteList _output))
  (if (> (Utilities|Array|Length _emotelist) 0)
    (Utilities|FlowControl|Sequence
      (:then_0
        (CallFunction|PlayEmote _emoteprimaryassetid_2_4fac2b88497d53109f4d1c94ce5db825 _emoteduration_5_17e1ada942eaa53ef5dd61aef468c9b6))
      (:then_1
        (Utilities|FlowControl|Delay _emoteduration_5_17e1ada942eaa53ef5dd61aef468c9b6)
        (Math|Integer|IncrementInt (Variables|Default|GetCurrentEmoteIndex)
          (:%20%20
            (if (< (Variables|Default|GetCurrentEmoteIndex) (Utilities|Array|Length (Variables|EmoteSetup|GetEmoteList)))
              (Utilities|FlowControl|Delay 0.2)
              (CallFunction|StartStory)
              (else
                (Variables|Default|SetCurrentEmoteIndex 0)
                (if (Variables|EmoteSetup|GetLoopAtEnd)
                  (Utilities|FlowControl|Delay (Variables|EmoteSetup|GetLoopDelay))
                  (CallFunction|StartStory) ; When done, start over!
                  (else
                    (Class|FortAthenaAIBotController|StopEmote (Variables|Default|GetAIBotControllerActivator))
                    (Utilities|FlowControl|Delay (Variables|EmoteSetup|GetEndDelay))
                    (CallFunction|ResetFocalPoint)
                    (Variables|Default|SetAIBotControllerActivator 0)
                    (LocalGameplayBehavior|FinishExecute)))))))))
    (else
      (Class|FortAthenaAIBotController|StopEmote (stale LocalGameplayBehavior|EventOnAbort _aibotcontrolleractivator))
      (CallFunction|ResetFocalPoint)
      (Variables|Default|SetAIBotControllerActivator 0)
      (LocalGameplayBehavior|FinishExecute))))

(event Custom|PlayEmote
  (params (Emote :type Primary%20Asset%20Id%20Structure) (Duration :type Float%20%28double-precision%29))
  ;[1 1 1] Play Emote
  (bind _aibotcontrolleractivator (Variables|Default|GetAIBotControllerActivator)
    (:then
      (Class|FortAthenaAIBotController|PlayEmote _aibotcontrolleractivator Emote)
      (Utilities|FlowControl|Delay Duration)
      (Class|FortAthenaAIBotController|StopEmote (Variables|Default|GetAIBotControllerActivator)))
    (:else
      (LocalGameplayBehavior|FinishExecute))))

(fn ResetFocalPoint
  (if (Variables|EmoteSetup|GetUseDirection)
    (bind _aibotcontrolleractivator (Variables|Default|GetAIBotControllerActivator)
      (:then
        (AI|ClearFocus _aibotcontrolleractivator))
      (:else))
    (else
      _))
  (return))

(fn SetFocalPoint
  (bind _aibotcontrolleractivator (Variables|Default|GetAIBotControllerActivator)
    (:then
      (AI|SetFocalPoint _aibotcontrolleractivator (+ (Transformation|GetActorLocation (Variables|/Script/FortniteAI.FortAthenaAIBotController|GetPlayerBotPawn _aibotcontrolleractivator)) (* (Transformation|GetActorForwardVector self) 1e+02))))
    (:else)))

(fn ConstructionScript)
